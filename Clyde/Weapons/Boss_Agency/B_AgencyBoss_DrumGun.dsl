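; Weapon Blueprint for the Agency boss's Drum Gun, a child class that adds no logic of its own.
; BeginPlay and the construction script only forward to the parent, and the parent Tick call is a disabled placeholder.
; The On Fire custom event receives a Fort Weapon reference but has an empty body.
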(event EventBeginPlay
  (|CallParentFunction))

(event EventTick
  (params (DeltaSeconds :type Float%20%28single-precision%29))
  (|CallParentFunction DeltaSeconds) ; This node is disabled and will not be called.\nDrag off pins to build functionality.
)

(event Custom|On%20Fire
  (params (Weapon :type Fort%20Weapon%20Object%20Reference)))

(fn ConstructionScript
  (|CallParentFunction))
